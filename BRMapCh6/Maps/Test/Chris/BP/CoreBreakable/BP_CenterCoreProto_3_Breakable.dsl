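; Breakable prototype center core piece (variant 3) for a Chapter 6 test map, which disappears when killed.
; On server death it waits a short random delay, then, only if its manual attachment actor is valid, destroys that actor and sets IsDead.
; The IsDead RepNotify hides the CenterCore_3 component without propagating to children and disables actor collision, replicating the broken state without destroying the core actor itself.
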(event EventOnDeathServer
  (params (Damage :type Float%20%28single-precision%29) (DamageTags :type Gameplay%20Tag%20Container%20Structure%20%28by%20ref%29) (Momentum :type Vector :default "0, 0, 0") (HitInfo :type Hit%20Result%20Structure%20%28by%20ref%29) (InstigatedBy :type Controller%20Object%20Reference) (DamageCauser :type Actor%20Object%20Reference) (EffectContext :type Gameplay%20Effect%20Context%20Handle%20Structure))
  (bind _manualattachmentactor (Variables|Default|GetManualAttachmentActor))
  (Utilities|FlowControl|Delay (Math|Random|RandomFloatinRange 0.2 0.3))
  (Utilities|IsValid _manualattachmentactor
    (:Is%20Valid
      (Actor|DestroyActor _manualattachmentactor)
      (Variables|Default|SetIsDead true))))

(fn OnRep_IsDead
  (if (Variables|Default|GetIsDead)
    (Rendering|SetVisibility (Variables|Default|GetCenterCore_3) false false)
    (Collision|SetActorEnableCollision false)))

(fn ConstructionScript)
